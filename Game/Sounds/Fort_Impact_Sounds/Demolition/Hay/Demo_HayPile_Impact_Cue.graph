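# Sound cue for hay pile impact hits, from the demolition impact sound set.
# A quality level node selects one of four branches: a random pick of three hay hit waves, two with a random pick of two, and one with a single wave.
# Each branch adds random pitch modulation.
# The cue uses the Fort_Impacts sound class, small-object demolition attenuation and a four-max impact concurrency set.
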SoundCue /Game/Sounds/Fort_Impact_Sounds/Demolition/Hay/Demo_HayPile_Impact_Cue
root=QualityLevel_0
nodes [
 {id=Modulator_0, children [Random_1], properties {pitchMax=1.2, pitchMin=0.92}}
 {id=Modulator_1, children [Random_2], properties {pitchMax=1.2, pitchMin=0.92}}
 {id=Modulator_2, children [Random_3], properties {pitchMax=1.2, pitchMin=0.92}}
 {id=Modulator_3, children [WavePlayer_4], properties {pitchMax=1.2, pitchMin=0.92}}
 {id=QualityLevel_0, children [Modulator_0,Modulator_1,Modulator_2,Modulator_3]}
 {id=Random_1, children [WavePlayer_148, WavePlayer_149, WavePlayer_150], properties {weights [1,1,1]}}
 {id=Random_2, children [WavePlayer_0,WavePlayer_1], properties {weights [1,1]}}
 {id=Random_3, children [WavePlayer_2,WavePlayer_3], properties {weights [1,1]}}
 {id=WavePlayer_0, properties {soundWaveAssetPtr=&/hay_hit_01}}
 {id=WavePlayer_1, properties {soundWaveAssetPtr=&/hay_hit_02}}
 {id=WavePlayer_148, properties {soundWaveAssetPtr=&hay_hit_01}}
 {id=WavePlayer_149, properties {soundWaveAssetPtr=&hay_hit_02}}
 {id=WavePlayer_150, properties {soundWaveAssetPtr=&/hay_hit_03}}
 {id=WavePlayer_2, properties {soundWaveAssetPtr=&hay_hit_01}}
 {id=WavePlayer_3, properties {soundWaveAssetPtr=&hay_hit_02}}
 {id=WavePlayer_4, properties {soundWaveAssetPtr=&hay_hit_01}}
]
properties {
 attenuationSettings=/Game/Sounds/Attenuation/Fort_Demolition_Hit_Small_Sized_Object_Att
 maxDistance=4000
 soundClassObject=&Fort_Impacts
 volumeMultiplier=0.476
 concurrencyOverrides {maxCount=4}
 concurrencySet [&CONC_Object_Impacts_FourMax]
}
