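# Sound cue for demolition impacts on hay, routed to the Fort_Impacts sound class.
# A mixer layers one of three random hay impacts, slightly pitch-varied, with one of three random default wood impacts at half input volume.
# The wood layer is pitched up by roughly half again.
# Small-object demolition attenuation applies, and a four-max object impact concurrency set is assigned.
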SoundCue /Game/Sounds/Fort_Impact_Sounds/Demolition/Hay/Demo_Hay_Impact_Cue
root=Mixer_7
nodes [
 {id=Mixer_7, children [Modulator_2,Modulator_16], properties {inputVolume [1,0.5]}}
 {id=Modulator_16, children [Random_46], properties {pitchMax=1.6, pitchMin=1.5}}
 {id=Modulator_2, children [Random_42], properties {pitchMax=1.07, pitchMin=0.92}}
 {id=Random_42, children [WavePlayer_130, WavePlayer_131, WavePlayer_132], properties {weights [1,1,1]}}
 {id=Random_46, children [WavePlayer_142, WavePlayer_143, WavePlayer_144], properties {weights [1,1,1]}}
 {id=WavePlayer_130, properties {soundWaveAssetPtr=&/Hay_Impact11}}
 {id=WavePlayer_131, properties {soundWaveAssetPtr=&/Hay_Impact10}}
 {id=WavePlayer_132, properties {soundWaveAssetPtr=&/Hay_Impact12}}
 {id=WavePlayer_142, properties {soundWaveAssetPtr=/Game/Sounds/Fort_Impact_Sounds/Demolition/Defaults/wood_impact_07}}
 {id=WavePlayer_143, properties {soundWaveAssetPtr=/Game/Sounds/Fort_Impact_Sounds/Demolition/Defaults/wood_impact_08}}
 {id=WavePlayer_144, properties {soundWaveAssetPtr=/Game/Sounds/Fort_Impact_Sounds/Demolition/Defaults/wood_impact_09}}
]
properties {
 attenuationSettings=/Game/Sounds/Attenuation/Fort_Demolition_Hit_Small_Sized_Object_Att
 maxDistance=4000
 soundClassObject=&Fort_Impacts
 volumeMultiplier=0.68
 concurrencyOverrides {maxCount=4, resolutionRule=StopOldest}
 concurrencySet [&CONC_Object_Impacts_FourMax]
}
